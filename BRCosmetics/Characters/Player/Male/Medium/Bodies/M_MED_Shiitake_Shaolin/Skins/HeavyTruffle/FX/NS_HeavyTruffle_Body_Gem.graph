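# Niagara cosmetic effect that adds a glowing, sparkling gem to the HeavyTruffle skin body of the Shiitake Shaolin character.
# The Gem emitter burst-spawns lifetime-looping particles drawn as mesh, rotating sprite and light, with curves pulsing warm color, glow size and light intensity.
# The Sparkles emitter continuously spawns short-lived sprites that drift with noise and drag while twinkling in size.
# Front-end modifiers adjust the Gem's alpha and sprite color between menus and gameplay.
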NiagaraSystem /BRCosmetics/Characters/Player/Male/Medium/Bodies/M_MED_Shiitake_Shaolin/Skins/HeavyTruffle/FX/NS_HeavyTruffle_Body_Gem
dependencies {
 usedDataInterfaces [&NiagaraDataInterfaceColorCurve, &NiagaraDataInterfaceCurve, &NiagaraDataInterfaceSkeletalMesh]
 usedDynamicInputs []
 usedModules [
  &self:SystemSpawnScript.SystemScriptSource.SystemScriptGraph.NiagaraNodeAssignment_0.SetVariables_#1
  &SystemState
 ]
}
emitters {
 Gem {
  emitterUpdateScript [
   &EmitterState(Life Cycle Mode=::NewEnumerator1, Loop Behavior=::NewEnumerator1, Loop Duration Mode=::NewEnumerator1, UseLoopDelay=false)
   &SpawnBurst_Instantaneous(Use Spawn Probability=false, UseLoopCountLimit=false)
  ]
  particleSpawnScript [
   &InitializeParticle(
    Color=<rgba 5 5 5 0.45>
    Mesh Scale Mode=::NewEnumerator3
    Position Offset=<0.6 -0.2 0>
    Sprite Size Mode=::NewEnumerator3
    Uniform Sprite Size=25
    UsePositionOffset=true
    Write Material Random=true
   )
   &InitialMeshOrientation(Mesh Orientation Mode=::NewEnumerator1, Rotation=<0 0 0.015>, Use Rotation Vector=true)
  ]
  particleUpdateScript [
   &ParticleState(
    Kill Particles When Lifetime Has Elapsed=false
    Let Infinitely Lived Particles Die When Emitter Deactivates=true
    Loop Particles Lifetime=true
   )
   &CameraOffset(Camera Offset Amount=0.25)
   &ScaleColor(
    Scale Alpha=&FloatFromCurve(
     FloatCurve=DataInterface() {propertyValues {Curve {keys [
         {arriveTangent=0.0924, interpMode=RCIM_Cubic, leaveTangent=0.0924, tangentMode=RCTM_User, time=0, value=0}
         {interpMode=RCIM_Cubic, tangentMode=RCTM_User, time=0.5, value=0.8}
         {arriveTangent=-0.3637, interpMode=RCIM_Cubic, leaveTangent=-0.3637, tangentMode=RCTM_User, time=1, value=0}
        ]}}}
     Scale Curve=&FN_FrontEndFloatModifier(Gameplay=1, Front End=1.1)
    )
    Scale RGB=&MakeVectorFromLinearColorRGB(Color=&ColorFromCurve(ColorCurve=DataInterface() {propertyValues {
        AlphaCurve {keys [<at 0 0.8 RCIM_Cubic>,<at 0.4999 0.9>,<at 0.9974 0.8>]}
        BlueCurve {keys [
          <at 0 0.0827 RCIM_Cubic>
          {arriveTangent=-0.0128, interpMode=RCIM_Cubic, leaveTangent=-0.0128, time=0.4848, value=0}
          <at 1 0.0699 RCIM_Cubic>
         ]}
        GreenCurve {keys [
          {arriveTangent=1.0015, interpMode=RCIM_Cubic, leaveTangent=1.0015, tangentMode=RCTM_User, time=0, value=0.3142}
          {arriveTangent=-0.0026, interpMode=RCIM_Cubic, leaveTangent=-0.0026, time=0.4848, value=0.1858}
          {arriveTangent=-0.7397, interpMode=RCIM_Cubic, leaveTangent=-0.7397, tangentMode=RCTM_User, time=1, value=0.3116}
         ]}
        RedCurve {keys [
          {arriveTangent=-0.9724, interpMode=RCIM_Cubic, leaveTangent=-0.9724, tangentMode=RCTM_User, time=0, value=1}
          <at 0.4848 1.6 RCIM_Cubic>
          {arriveTangent=0.7691, interpMode=RCIM_Cubic, leaveTangent=0.7691, tangentMode=RCTM_User, time=1, value=1}
         ]}
       }}))
    ScaleA=true
    ScaleRGB=true
   )
   &FN_Sprite_Attributes(
    Color=&MultiplyLinearColor(
     Linear Color=&ColorFromCurve(ColorCurve=DataInterface() {propertyValues {
        AlphaCurve {keys [<at 0.0006 0.55>,<at 0.4904 0.8 RCIM_Cubic>,<at 0.9929 0.55>]}
        BlueCurve {keys [<at 0 0 RCIM_Cubic>,<at 0.4974 0.0401>,<at 1 0 RCIM_Cubic>]}
        GreenCurve {keys [
          {arriveTangent=1.3771, interpMode=RCIM_Cubic, time=0, value=0.4789}
          <at 0.4974 0.3725>
          {interpMode=RCIM_Cubic, leaveTangent=-1.3628, time=1, value=0.4789}
         ]}
        RedCurve {keys [
          {arriveTangent=-1.2835, interpMode=RCIM_Cubic, time=0, value=0.7913}
          <at 0.4974 1>
          {interpMode=RCIM_Cubic, leaveTangent=1.2702, time=1, value=0.7913}
         ]}
       }})
     Scale Factor=&MakeVector4_FromFloat(Float=&FN_FrontEndFloatModifier(Gameplay=1, Front End=0.6))
    )
    Create Color Attribute=true
    Create SpriteSize Attribute=true
    SpriteSize=&Vector2DFromFloat(Value=&FloatFromCurve(
      FloatCurve=DataInterface() {propertyValues {Curve {keys [
          {interpMode=RCIM_Cubic, tangentMode=RCTM_User, time=0, value=0.5}
          {interpMode=RCIM_Cubic, tangentMode=RCTM_User, time=0.5, value=1}
          {tangentMode=RCTM_User, time=1, value=0.5}
         ]}}}
      Scale Curve=35
     ))
    Write to Module Namespace=true
   )
   &SpriteRotationRate(Rotation Rate=30)
   &FN_Light_Attributes(
    Color=$Particles.Color
    Create Color Attribute=true
    Create Exponent Attribute=true
    Create Position Attribute=true
    Create Radius Attribute=true
    Exponent=5
    Radius=10
    Scale Color=&FloatFromCurve(FloatCurve=DataInterface() {propertyValues {Curve {keys [
         {interpMode=RCIM_Cubic, tangentMode=RCTM_User, time=0, value=0.5}
         {interpMode=RCIM_Cubic, tangentMode=RCTM_User, time=0.5, value=1}
         {tangentMode=RCTM_User, time=1, value=0.5}
        ]}}})
   )
  ]
  rendererClasses [
   &NiagaraSpriteRendererProperties
   &NiagaraMeshRendererProperties
   &NiagaraLightRendererProperties
  ]
  renderers [
   {rendererClass=&NiagaraSpriteRendererProperties}
   {rendererClass=&NiagaraMeshRendererProperties, rendererIndex=1}
   {rendererClass=&NiagaraLightRendererProperties, rendererIndex=2}
  ]
 }
 Sparkles {
  emitterUpdateScript [
   &EmitterState()
   &SpawnRate(SpawnRate=5, Use Spawn Probability=false)
  ]
  particleSpawnScript [
   &InitializeParticle(
    Color=<rgba 1 1 1 0.8>
    Lifetime Max=1.5
    Lifetime Mode=::NewEnumerator1
    Position Offset=&RandomRangeVector(Minimum=<-2 -1 -1>, Maximum=<2 1 1>)
    Sprite Size Mode=::NewEnumerator1
    Uniform Sprite Size Max=15
    UsePositionOffset=true
    Write Material Random=true
   )
   &AddVelocity(Velocity=&RandomRangeVector(Minimum=<-10 -5 -10>, Maximum=<10 -10 10>))
  ]
  particleUpdateScript [
   &ParticleState()
   &ScaleColor(
    Scale Alpha=&FloatFromCurve(FloatCurve=DataInterface() {propertyValues {Curve {keys [
         {arriveTangent=3.3997, interpMode=RCIM_Cubic, leaveTangent=3.3997, tangentMode=RCTM_User, time=0, value=0}
         {interpMode=RCIM_Cubic, tangentMode=RCTM_User, time=0.5, value=1}
         {arriveTangent=-2.7937, interpMode=RCIM_Cubic, leaveTangent=-2.7937, tangentMode=RCTM_User, time=1, value=0}
        ]}}})
    Scale RGB=&MakeVectorFromLinearColorRGB(Color=&ColorFromCurve(ColorCurve=DataInterface() {propertyValues {
        AlphaCurve {keys [<at 0 1 RCIM_Cubic>]}
        BlueCurve {keys [<at 0 0.313 RCIM_Cubic>, <at 0.4848 0 RCIM_Cubic>, {interpMode=RCIM_Cubic, leaveTangent=-0.9429, time=1, value=0.313}]}
        GreenCurve {keys [
          {arriveTangent=1.0015, interpMode=RCIM_Cubic, leaveTangent=1.0015, tangentMode=RCTM_User, time=0, value=0.779}
          <at 0.4848 0.0935 RCIM_Cubic>
          {arriveTangent=-0.7397, interpMode=RCIM_Cubic, leaveTangent=-0.7397, tangentMode=RCTM_User, time=1, value=0.779}
         ]}
        RedCurve {keys [
          {arriveTangent=-0.9724, interpMode=RCIM_Cubic, leaveTangent=-0.9724, tangentMode=RCTM_User, time=0, value=1}
          <at 0.4848 0.6445 RCIM_Cubic>
          {arriveTangent=0.7691, interpMode=RCIM_Cubic, leaveTangent=0.7691, tangentMode=RCTM_User, time=1, value=1}
         ]}
       }}))
    ScaleA=true
    ScaleRGB=true
   )
   &ScaleSpriteSize(Uniform Curve Sprite Scale=DataInterface() {propertyValues {Curve {keys [
        {arriveTangent=4, interpMode=RCIM_Cubic, leaveTangent=4, time=0, value=0}
        {arriveTangent=4, interpMode=RCIM_Cubic, leaveTangent=4, time=0.1, value=0.4}
        {arriveTangent=-0.6667, interpMode=RCIM_Cubic, leaveTangent=-0.6667, time=0.2, value=0.8}
        {arriveTangent=-1.3333, interpMode=RCIM_Cubic, leaveTangent=-1.3333, time=0.4, value=0.2}
        <at 0.5 0.4 RCIM_Cubic>
        {arriveTangent=2, interpMode=RCIM_Cubic, leaveTangent=2, time=0.6, value=0.2}
        {arriveTangent=-0.5, interpMode=RCIM_Cubic, leaveTangent=-0.5, time=0.8, value=1}
        {arriveTangent=-5, interpMode=RCIM_Cubic, leaveTangent=-5, time=1, value=0}
       ]}}})
   &VectorNoiseForce(Force Amount=50, Module.Use Random Noise Time Interval=false)
   &Drag(Drag=2, Use Linear Drag=true, Use Rotational Drag=false)
   &SolveForcesAndVelocity(Clamp Velocity=false, Limit Acceleration=false)
  ]
  rendererClasses [&NiagaraSpriteRendererProperties]
  renderers [{rendererClass=&NiagaraSpriteRendererProperties}]
 }
}
summary {
 systemName=NS_HeavyTruffle_Body_Gem
 emitters [
  {
   emitterName=Gem
   rendererClasses [
    &NiagaraSpriteRendererProperties
    &NiagaraMeshRendererProperties
    &NiagaraLightRendererProperties
   ]
  }
  {emitterName=Sparkles, rendererClasses [&NiagaraSpriteRendererProperties]}
 ]
 userVariables [
  User.bIsFrontEnd {type=NiagaraBool}
  User.bIsFrontEndPreview {type=NiagaraBool}
 ]
}
